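# Unlit hologram material for the Day Trader backpack, usable on skeletal and Niagara meshes.
# Vertex red drives an inner ring glow, while vertex green masks procedural dots made from tiled, panning UVs and a distance field with a gradient fade.
# Particle color tints both layers.
# Dynamic-parameter swipe and pulse values modulate the green-masked area before MF_FX_EmissiveAdjustments processes the emissive.
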Material /BRCosmetics/Accessories/FORT_Backpacks/Backpack_MED_Day_Trader/Materials/M_Hologram_Daytrader
graph {
 properties {
  allowTranslucentCustomDepthWrites=true
  bUsedWithMeshDeformer=true
  bUsedWithMorphTargets=true
  bUsedWithNiagaraMeshParticles=true
  bUsedWithSkeletalMesh=true
  pixelDepthOffsetMode=PDOM_Legacy
  shadingModel=MSM_Unlit
  usedShadingModels=Unlit
 }
 nodes [
  n1 = VertexColor()
  n2 = Multiply(A=@n1.R, B=ScalarParameter() {defaultValue=10, group=InnerRing, parameterName=Glow Intesity}) {constB=10}
  n3 = TextureCoordinate()
  n4 = ScalarParameter() {defaultValue=48, group=Dots, parameterName=Tiling}
  n5 = Divide(A=Time(), B=Max(A=@n4))
  n6 = AppendVector(A=Constant(), B=ScalarParameter() {defaultValue=1, group=Dots, parameterName=Pan Speed})
  n7 = Frac(None=Multiply(A=Panner(Coordinate=@n3, Time=@n5, Speed=@n6) {speedY=1}, B=@n4))
  n8 = Distance(A=Constant(), B=Multiply(A=Add(B=@n7) {constA=-0.5, constB=-0.5}, B=Constant() {r=2}) {constB=2})
  n9 = ScalarParameter() {defaultValue=0.85, group=Dots, parameterName=Size}
  n10 = Saturate(None=Add(
    A=MaterialFunctionCall() {materialFunction=&LinearGradient}.VGradient
    B=ScalarParameter() {group=Dots, parameterName=Fade Offset, sliderMax=1}
   ))
  n11 = LinearInterpolate(
   B=MaterialFunctionCall(In=Power(Base=@n8, Exp=@n9) {constExponent=1}, Contrast=ScalarParameter() {defaultValue=1, group=Dots, parameterName=Sharpness}) {materialFunction=&CheapContrast}.Result
   Alpha=Multiply(A=@n10, B=@n10)
  ) {constA=1}
  n12 = ScalarParameter() {defaultValue=0.25, group=Dots, parameterName=Dimmer, sliderMax=1}
  n13 = Multiply(A=@n1.G, B=Clamp(None=Subtract(A=@n11, B=@n12) {constB=0.1}) {minDefault=0.35})
  n14 = Multiply(A=ParticleColor().RGB, B=Add(A=Add(A=@n2, B=@n13)))
  n15 = DynamicParameter() {defaultValue=<rgba 0 1 1 1>, paramNames [Emissive_Swipe, Emissive_Pulse, 0, 0]}
  n16 = MaterialFunctionCall() {materialFunction=&LinearGradient}
  n17 = LinearInterpolate(A=Add(A=SmoothStep(Min=@n15.Emissive_Swipe, Value=@n16.VGradient)), B=Add(A=@n15.Emissive_Pulse), Alpha=@n15.Emissive_Pulse)
  MaterialFunctionCall(Emissive=LinearInterpolate(A=@n14, B=Multiply(A=@n14, B=@n17), Alpha=VertexColor().G)) {materialFunction=&MF_FX_EmissiveAdjustments}
 ]
}
