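# Idle Niagara effect for the Day Trader backpack: a mesh arrow and a soft glow sprite, both following the backpack's skeletal mesh.
# Each emitter bursts one looping particle; the arrow bobs on a sine offset and cycles a material parameter through a curve.
# User on-kill floats squash, stretch and brighten the arrow.
# The glow's alpha pulses gently, and its blue tint depends on the frontend flag.
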NiagaraSystem /BRCosmetics/Accessories/FORT_Backpacks/Backpack_MED_Day_Trader/NS_Backpack_Daytrader_Idle
dependencies {usedDataInterfaces [&NiagaraDataInterfaceCurve], usedDynamicInputs [], usedModules [&SystemState]}
emitters {
 Arrow_Not_Broken {
  emitterUpdateScript [
   &EmitterState(Life Cycle Mode=::NewEnumerator1, Loop Behavior=::NewEnumerator1, Loop Duration Mode=::NewEnumerator1, UseLoopDelay=false)
   &SpawnBurst_Instantaneous(Spawn Count=1, Use Spawn Probability=false)
  ]
  particleSpawnScript [
   &InitializeParticle(
    Color=<rgba 1 0.0667 0 1>
    Lifetime=5
    Mesh Scale=&Multiply_Vector(A=&VectorFromFloat(Value=1.1), B=<0.5 1 1>)
    Mesh Scale Mode=::NewEnumerator4
    UsePositionOffset=false
    Write Material Random=true
   )
   &InitialMeshOrientation(Rotation=<0 0 0.25>, Use Orientation Vector=true, Use Rotation Vector=true)
   /SetVariables(Particles.MeshOrientation=&MultiplyQuaternion(
     Quaternion A=$Particles.MeshOrientation
     Quaternion B=&MakeQuatFromAxisAngle(Axis=<1 0 0>, AngleInDegrees=-33)
    )) {isSetParameters=true}
  ]
  particleUpdateScript [
   &ParticleState(
    Kill Particles When Lifetime Has Elapsed=false
    Let Infinitely Lived Particles Die When Emitter Deactivates=true
    Loop Particles Lifetime=true
   )
   &ScaleMeshSize(Scale Factor=&MakeVector(
     X=&Add_Float(A=1, B=&Multiply_Float(A=-0.25, B=$User.OnKill_Bounce_Delay))
     Y=&Add_Float(A=1, B=&Multiply_Float(A=0.25, B=$User.OnKill_Bounce))
     Z=&Add_Float(A=1, B=&Multiply_Float(A=0.25, B=$User.OnKill_Bounce_Delay))
    ))
   &SkeletalMeshLocation(
    Bone Sampling Mode=::NewEnumerator4
    Inherit Velocity=false
    MaskSpawnGroup=false
    Sampled Position Offset=&RotateVectorByQuaternion(
     Vector To Rotate=&MakeVector(X=-5, Y=0, Z=&Sine(Normalized Angle=$Emitter.Age, Period=1, Scale=2, Bias=4))
     Quaternion=$Particles.MeshOrientation
    )
    Sampled Position Offset Coordinate Space:ENiagaraMeshTransforms=::NewEnumerator3
    Skeletal Mesh=$User.SkeletalMesh
   )
   &DynamicMaterialParameters(
    Index 0 Param 1=&FloatFromCurve(
     FloatCurve=DataInterface() {propertyValues {Curve {keys [
         <at 0.1 0.9 RCIM_Cubic>
         {arriveTangent=-2.2219, time=0.45, value=0.25}
         {arriveTangent=0.9286, interpMode=RCIM_Cubic, leaveTangent=0.9286, tangentMode=RCTM_Break, time=0.7, value=0.25}
         {interpMode=RCIM_Cubic, leaveTangent=4.8461, time=1, value=0.9}
        ]}}}
     CurveIndex=&Frac_Float(Float=&Multiply_Float(A=1, B=&Add_Float(A=0.5, B=&Multiply_Float(A=1, B=$Emitter.Age))))
    )
    Module.Param0WriteEnabled=true
    Module.Param1WriteEnabled=false
    Module.Param2WriteEnabled=false
    Module.Param3WriteEnabled=false
   )
   &ScaleColor(
    Scale RGB=&VectorFromFloat(Value=&Add_Float(A=1, B=&Multiply_Float(A=1, B=$User.OnKill_Pulse)))
    ScaleA=true
    ScaleRGB=true
   )
  ]
  rendererClasses [&NiagaraMeshRendererProperties]
  renderers [{rendererClass=&NiagaraMeshRendererProperties, rendererIndex=0}]
 }
 GlowSprite {
  emitterUpdateScript [
   &EmitterState(Life Cycle Mode=::NewEnumerator1, Loop Behavior=::NewEnumerator1, Loop Duration Mode=::NewEnumerator1, UseLoopDelay=false)
   &SpawnBurst_Instantaneous(Spawn Count=1, Use Spawn Probability=false)
  ]
  particleSpawnScript [&InitializeParticle(
    Color=&MakeCustomLinearColorFromBool(Bool=$User.bIsFrontend, True Color=<rgba 0.1 0.3 1 1>, False Color=<rgba 0.05 0.15 0.5 1>)
    Sprite Size Mode=::NewEnumerator3
    Uniform Sprite Size=40
    UsePositionOffset=false
    Write Material Random=true
   )]
  particleUpdateScript [
   &ParticleState(
    Kill Particles When Lifetime Has Elapsed=false
    Let Infinitely Lived Particles Die When Emitter Deactivates=true
    Loop Particles Lifetime=true
   )
   &ScaleColor(
    Scale Alpha=&Lerp_Float(A=0.25, B=0.3, Alpha=&Sine(Normalized Angle=$Emitter.Age, Period=2.5, Scale=1, Bias=0))
    Scale RGB=&VectorFromFloat(Value=1)
    ScaleA=true
    ScaleRGB=true
   )
   &SkeletalMeshLocation(
    Bone Sampling Mode=::NewEnumerator4
    Inherit Velocity=false
    MaskSpawnGroup=false
    Sampled Position Offset=<0 -3.5 0>
    Sampled Position Offset Coordinate Space:ENiagaraMeshTransforms=::NewEnumerator3
    Skeletal Mesh=$User.SkeletalMesh
   )
   /NiagaraFN/Sprite/FN_SpriteAlignmentAndFacing(SpriteFacing:Vector3f=&Multiply_VectorByFloat(Vector=$Engine.Owner.SystemYAxis, Float=-1)) {enabled=false}
  ]
  rendererClasses [&NiagaraSpriteRendererProperties]
  renderers [{rendererClass=&NiagaraSpriteRendererProperties}]
 }
}
summary {
 systemName=NS_Backpack_Daytrader_Idle
 emitters [
  {emitterName=Arrow_Not_Broken, rendererClasses [&NiagaraMeshRendererProperties]}
  {emitterName=GlowSprite, rendererClasses [&NiagaraSpriteRendererProperties]}
 ]
 userVariables [
  User.bIsFrontend {type=NiagaraBool}
  User.InWater {type=NiagaraBool}
  User.OnKill_Bounce {type=NiagaraFloat, defaultValue=0, description=""}
  User.OnKill_Bounce_Delay {type=NiagaraFloat, defaultValue=0, description=""}
  User.OnKill_Pulse {type=NiagaraFloat, defaultValue=0, description=""}
  User.OnKill_Swipe {type=NiagaraFloat, defaultValue=0, description=""}
  User.SkeletalMesh {
   type=NiagaraDataInterfaceSkeletalMesh
   defaultValue {
    struct=NiagaraExt_VariableValue_DataInterface
    value {dataInterface=&self:NiagaraDataInterfaceSkeletalMesh_1}
   }
  }
 ]
}
